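# Physics asset for the male Parcel Prank Hammer pickaxe's skeletal mesh, giving its Dyn_HammerFace bones collision and jointed physics.
# Box shapes cover three bodies, with a compound of three angled boxes on Dyn_HammerFace2.
# Two constraints chain Dyn_HammerFaceMSTR to Dyn_HammerFace1 to Dyn_HammerFace2, with moderate swing limits and tight twist so the head can wobble without spinning.
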PhysicsAsset /BRCosmetics/Weapons/FORT_Melee/Pickaxe_Parcel_Prank_Hammer_Male/Meshes/SK_Parcel_PrankHammer_Male_Physics
bodies {
 Dyn_HammerFace1 [{shapeType=Box, extentX=8.1875, extentY=4.8125, extentZ=26.75, center=<4.3791 0 0>, rotation=<rot 0 0 -45.0011>}]
 Dyn_HammerFace2 [
  {shapeType=Box, extentX=13.75, extentY=5.9375, extentZ=26.75, center=<4.7041 -0.4144 -0.0002>, rotation=<rot 0 -5.6309 -47.8138>}
  {shapeType=Box, extentX=2.4375, extentY=20.4375, extentZ=28.9375, center=<2.4474 5.7681 8.5011>, rotation=<rot 0 -5.6272 -47.8074>}
  {shapeType=Box, extentX=2.5, extentY=11.4375, extentZ=5, center=<0.9074 -7.8424 -8.2037>, rotation=<rot 0 -5.6262 47.8112>}
 ]
 Dyn_HammerFaceMSTR [{shapeType=Box, extentX=2, extentY=30.375, extentZ=30.3225, center=<-0.6807 0.101 -0.0001>, rotation=<rot 0 -8.4444 0>}]
}
constraints [
 {bone1Name=Dyn_HammerFace2, bone2Name=Dyn_HammerFace1, swing1LimitDegrees=25, swing2LimitDegrees=25, twistLimitDegrees=5}
 {bone1Name=Dyn_HammerFace1, bone2Name=Dyn_HammerFaceMSTR, swing1LimitDegrees=20, swing2LimitDegrees=20, twistLimitDegrees=5}
]
