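# Animation Blueprint for the male Parcel Prank Hammer pickaxe that adds secondary physics motion to the hammer face.
# Both branches run RigidBody simulation in base-bone space, then a spring-driven AnimDynamics on Dyn_HammerFace2 relative to Dyn_HammerFace1.
# Jumping, falling or sprinting selects the branch based on b_AR_Root.
# Otherwise, the Melee_TwoHanded_Handle branch plays, with tighter linear limits and an additive offset on Dyn_HammerFaceMSTR.
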AnimBlueprint /BRCosmetics/Weapons/FORT_Melee/Pickaxe_Parcel_Prank_Hammer_Male/Meshes/SK_Parcel_PrankHammer_Male_AnimBlueprint
target_skeleton=&/SK_Parcel_PrankHammer_Male_Skeleton
graphs [
 AnimGraph {nodes [
   n1 = RigidBody(ComponentPose=Local To Component(LocalPose=Input Pose() {name=InPose_0})) {
    bOverrideWorldGravity=true
    overridePhysicsAsset=&/SK_Parcel_PrankHammer_Male_Physics
    simulationSpace=BaseBoneSpace
    alphaBoolBlend {blendOption=Linear}
    baseBoneRef {boneName=b_AR_Root}
    componentLinearAccScale=<0.4 0.4 0.4>
    componentLinearVelScale=<0.7 0.7 0.7>
    overrideWorldGravity {z=-980}
   }
   n2 = Anim Dynamics - Bone: Dyn_HammerFace2(ComponentPose=@n1, Alpha=0.7, ExternalForce="0,0,0") {
    angularDampingOverride=0.9
    angularSpringConstant=150
    bAngularSpring=true
    bGravityOverrideInSimSpace=true
    bLinearSpring=true
    bOverrideAngularDamping=true
    bOverrideLinearDamping=true
    bUseGravityOverride=true
    linearDampingOverride=0.95
    linearSpringConstant=200
    simulationSpace=BoneRelative
    alphaBoolBlend {blendOption=Linear}
    boundBone {boneName=Dyn_HammerFace2}
    componentLinearAccScale=<0.3 0.3 0.3>
    componentLinearVelScale=<0.6 0.6 0.6>
    gravityOverride {z=-980}
    physicsBodyDefinitions [{
      sphereCollisionRadius=0
      boundBone {boneName=Dyn_HammerFace2}
      boxExtents=<10 10 10>
      constraintSetup {angularLimitsMax=<0 0.3 0.3>, angularLimitsMin=<0 -0.3 -0.3>, angularTarget=<0 0 0>, linearAxesMax=<3 1 1>, linearAxesMin=<-3 -1 -1>}
      localJointOffset=<0 -3 -5>
     }]
    relativeSpaceBone {boneName=Dyn_HammerFace1}
   }
   n3 = RigidBody(ComponentPose=Local To Component(LocalPose=Input Pose())) {
    bOverrideWorldGravity=true
    overridePhysicsAsset=&SK_Parcel_PrankHammer_Male_Physics
    simulationSpace=BaseBoneSpace
    alphaBoolBlend {blendOption=Linear}
    baseBoneRef {boneName=Melee_TwoHanded_Handle}
    componentLinearAccScale=<0.3 0.3 0.3>
    componentLinearVelScale=<0.6 0.6 0.6>
    overrideWorldGravity {z=-980}
   }
   n4 = Anim Dynamics - Bone: Dyn_HammerFace2(ComponentPose=@n3, Alpha=0.7, ExternalForce="0,0,0") {
    angularDampingOverride=0.9
    angularSpringConstant=150
    bAngularSpring=true
    bGravityOverrideInSimSpace=true
    bLinearSpring=true
    bOverrideAngularDamping=true
    bOverrideLinearDamping=true
    bUseGravityOverride=true
    linearDampingOverride=0.95
    linearSpringConstant=200
    simulationSpace=BoneRelative
    alphaBoolBlend {blendOption=Linear}
    boundBone {boneName=Dyn_HammerFace2}
    componentLinearAccScale=<0.3 0.3 0.3>
    componentLinearVelScale=<0.6 0.6 0.6>
    gravityOverride {z=-980}
    physicsBodyDefinitions [{
      sphereCollisionRadius=0
      boundBone {boneName=Dyn_HammerFace2}
      boxExtents=<10 10 10>
      constraintSetup {angularLimitsMax=<0 0.3 0.3>, angularLimitsMin=<0 -0.3 -0.3>, angularTarget=<0 0 0>, linearAxesMax=<3 0.3 0.3>, linearAxesMin=<-2 -0.3 -0.3>}
      localJointOffset=<0 -3 -5>
     }]
    relativeSpaceBone {boneName=Dyn_HammerFace1}
   }
   n5 = "Transform (Modify) Bone - Bone: Dyn_HammerFaceMSTR"(ComponentPose=@n4, Alpha=1, Translation="0,2.5,0", Rotation="0,7,3", Scale="1,1,1") {
    rotationMode=BMM_Additive
    translationMode=BMM_Additive
    alphaBoolBlend {blendOption=Linear}
    boneToModify {boneName=Dyn_HammerFaceMSTR}
    rotation {roll=3, yaw=7}
    translation {y=2.5}
   }
   n6 = Get bIsFalling()
   n7 = Get bIsJumping()
   n8 = Get bIsSprinting()
   Output Pose(Result=Blend Poses by bool(
     BlendPose_0=Component To Local(ComponentPose=@n2)
     BlendPose_1=Component To Local(ComponentPose=@n5)
     bActiveValue=OR Boolean!(A=@n7, B=@n6, C:bool=@n8) -> (ReturnValue)
    ) {blendType=Linear}) {layerGroup=None}
  ]}
 EventGraph {nodes [
   TryGetPawnOwner()
   Event BlueprintUpdateAnimation()
  ]}
]
variables []
